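# Masked, unlit, two-sided Niagara material for a swirling cone effect in the FrogLTE2 boost.
# Panning Perlin noise in a radial gradient warps cone-splat UVs about their center by the Distort parameter.
# Particle color tints the red channel, plus a green-channel shine sharpened by Shines, for emissive.
# Panning curl noise minus Erode breaks up the red-channel mask, while time-driven sine and cosine build an orbit offset vector.
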Material /VehicleCosmetics/Boosts/Boost_FrogLTE2/Material/M_Boost_FrogLTE2_ConeSwirl01
graph {
 properties {
  bAutomaticallySetUsageInEditor=false
  bUsedWithNiagaraMeshParticles=true
  bUsedWithNiagaraSprites=true
  blendMode=BLEND_Masked
  shadingModel=MSM_Unlit
  twoSided=true
  usedShadingModels=Unlit
 }
 nodes [
  MaterialExpressionMeshEmitterVertexColor_0 = ParticleColor()
  n1 = TextureCoordinate()
  n2 = ScalarParameter() {defaultValue=1, parameterName=tile}
  n3 = Divide(A=@n2, B=Constant() {r=2})
  n4 = Constant() {r=-1}
  n5 = Panner(Coordinate=Add(A=Multiply(A=@n1, B=@n2), B=Multiply(A=@n3, B=@n4))) {speedY=-1}
  n6 = Multiply(
   A=MaterialFunctionCall() {materialFunction=&RadialGradientExponential}.RadialGradientExponential
   B=TextureSample(UVs=@n5) {texture=/DealershipSharedLibrary/FX/Textures/Noise/T_VCVFX_PerlinBlur}.R
  )
  MaterialExpressionMeshEmitterDynamicParameter_0 = DynamicParameter() {paramNames [Distort,Erode,Shines,Param4]}
  n7 = ComponentMask(None=TextureCoordinate()) {r=true}
  n8 = Constant() {r=0.5}
  n9 = ComponentMask(None=TextureCoordinate()) {g=true}
  n10 = Constant() {r=0.5}
  n11 = AppendVector(
   A=Multiply(A=@MaterialExpressionMeshEmitterDynamicParameter_0.Distort, B=Subtract(A=@n7, B=@n8))
   B=Multiply(A=@MaterialExpressionMeshEmitterDynamicParameter_0.Distort, B=Subtract(A=@n9, B=@n10))
  )
  n12 = TextureSample(UVs=Add(A=TextureCoordinate(), B=Multiply(A=@n6, B=@n11))) {texture=/DealershipSharedLibrary/FX/Textures/Gradient_Masks/T_VCVFX_ConeSplat}
  n13 = Desaturation(None=@MaterialExpressionMeshEmitterVertexColor_0.RGB, Fraction=Constant() {r=0.9})
  n14 = Power(Base=Multiply(A=Constant3Vector() {constant=<rgba 0.03 0.7 0.08 0>}, B=@n12.G), Exp=@MaterialExpressionMeshEmitterDynamicParameter_0.Shines)
  MaterialFunctionCall(Emissive=Add(A=Multiply(A=@MaterialExpressionMeshEmitterVertexColor_0.RGB, B=@n12.R), B=Multiply(A=@n13, B=@n14))) {materialFunction=&MF_FX_EmissiveAdjustments}
  n15 = Multiply(A=Time(), B=ScalarParameter() {parameterName=OrbitVelocity})
  n16 = Sine(None=@n15)
  n17 = ScalarParameter() {parameterName=OrbitRadius}
  n18 = Multiply(A=Cosine(None=@n15), B=ScalarParameter() {parameterName=OrbRadiusY})
  Multiply(
   A=AppendVector(A=Multiply(A=@n16, B=@n17), B=AppendVector(A=@n16, B=@n18))
   B=ScalarParameter() {defaultValue=15, parameterName=OrbitStrenght}
  )
  n19 = Constant() {r=2}
  n20 = TextureSample(UVs=@n5) {texture=/DealershipSharedLibrary/FX/Textures/Noise/T_VCVFX_Flow_Curl_01}
  Reroute(None=Multiply(A=Multiply(A=@n12.R, B=@n19), B=Subtract(A=@n20.G, B=@MaterialExpressionMeshEmitterDynamicParameter_0.Erode)))
 ]
}
