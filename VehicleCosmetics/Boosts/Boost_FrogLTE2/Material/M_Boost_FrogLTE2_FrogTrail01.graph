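# Masked unlit Niagara ribbon and sprite material for the FrogLTE2 vehicle boost's frog trail.
# A panning trail texture, driven by dynamic speed and offset, supplies a yellow shimmer highlight and a particle-colored body, broken up by scrolling shimmer noise with dynamic erode, feeding an emissive adjustment function.
# A cutout mask subtracts panning frog-mask textures, one noise-distorted, and a UV edge term from the trail's blue channel.
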Material /VehicleCosmetics/Boosts/Boost_FrogLTE2/Material/M_Boost_FrogLTE2_FrogTrail01
graph {
 properties {
  bAutomaticallySetUsageInEditor=false
  bUsedWithNiagaraRibbons=true
  bUsedWithNiagaraSprites=true
  blendMode=BLEND_Masked
  shadingModel=MSM_Unlit
  usedShadingModels=Unlit
 }
 nodes [
  n1 = ComponentMask(None=TextureCoordinate()) {r=true}
  n2 = OneMinus(None=@n1)
  n3 = Constant() {r=0.005}
  MaterialExpressionConstantClamp_11 = Saturate(None=Ceil(None=Subtract(A=@n2, B=@n3)))
  n4 = TextureSample(UVs=Panner(Coordinate=TextureCoordinate()) {speedY=0.5}) {samplerSource=SSM_Wrap_WorldGroupSettings, texture=/DealershipSharedLibrary/FX/Textures/TexturePackage/T_VCVFX_MaskFrog_RGB}
  n5 = Constant() {r=-0.9}
  MaterialExpressionConstantClamp_12 = Saturate(None=Ceil(None=Subtract(A=@n1, B=@n5)))
  n6 = DynamicParameter() {paramNames [speed, offSET, offset mask, erode]}
  n7 = AppendVector(A=@n6.offset mask, B=ScalarParameter() {parameterName=append mask})
  n8 = TextureCoordinate() {vTiling=2}
  n9 = TextureCoordinate() {uTiling=0.5, vTiling=0.5}
  n10 = ComponentMask(None=TextureSample(UVs=Panner(Coordinate=@n9) {speedX=1, speedY=1}) {texture=/DealershipSharedLibrary/FX/Textures/Noise/T_VCVFX_Alpine_Infinite_RGBNoise01}.RGB) {g=true}
  n11 = Constant() {r=0.18}
  n12 = Panner(Coordinate=Add(A=Add(A=@n7, B=@n8), B=Multiply(A=@n10, B=@n11))) {speedY=1.4}
  n13 = Multiply(
   A=@MaterialExpressionConstantClamp_12
   B=Power(
    Base=TextureSample(UVs=@n12) {samplerSource=SSM_Wrap_WorldGroupSettings, texture=&T_VCVFX_MaskFrog_RGB}.R
    Exp=Constant() {r=0.44}
   )
  )
  MaterialExpressionConstantClamp_10 = Saturate(None=Add(A=OneMinus(None=@MaterialExpressionConstantClamp_11), B=Add(A=@n4.R, B=@n13)))
  n14 = AppendVector(A=@n6.offSET, B=ScalarParameter() {parameterName=append param})
  n15 = TextureCoordinate()
  n16 = Time()
  n17 = TextureSample(UVs=Panner(Coordinate=Add(A=@n14, B=@n15), Time=Multiply(A=@n6.speed, B=@n16)) {speedX=-2}) {texture=/DealershipSharedLibrary/FX/Textures/TexturePackage/T_VCVFX_FrogTrail01_RGB}
  MaterialExpressionConstantClamp_7 = Saturate(None=Subtract(A=@n17.B, B=@MaterialExpressionConstantClamp_10))
  n18 = Constant() {r=17}
  n19 = Constant() {r=1}
  n20 = Power(Base=Multiply(A=Multiply(A=@n17.G, B=@n19), B=@n17.G), Exp=Constant() {r=0.2})
  MaterialExpressionConstantClamp_8 = Saturate(None=Multiply(A=Multiply(A=@n17.G, B=@n18), B=DotProduct(A=@n20, B=@n17.R)))
  MaterialExpressionVertexColor_1 = ParticleColor()
  n21 = Multiply(A=@MaterialExpressionConstantClamp_8, B=@MaterialExpressionVertexColor_1.A)
  n22 = VectorParameter() {parameterName=yellow shimmer, defaultValue=<rgba 0.6077 0.8202 0 0>}
  n23 = Multiply(A=@MaterialExpressionVertexColor_1.RGB, B=@n17.R)
  n24 = Time()
  n25 = ScalarParameter() {defaultValue=1, parameterName=NoiseSpeed}
  n26 = TextureSample(UVs=Panner(Coordinate=TextureCoordinate(), Time=Multiply(A=@n24, B=@n25)) {speedX=-2}) {texture=/DealershipSharedLibrary/FX/Textures/Diffuse/T_VCVFX_Frog_Shimmer_RGB}
  MaterialExpressionConstantClamp_9 = Saturate(None=OneMinus(None=Subtract(A=@n26.G, B=@n6.erode)))
  MaterialFunctionCall(Emissive=Add(A=Multiply(A=@n21, B=@n22.RGB), B=Multiply(A=@n23, B=@MaterialExpressionConstantClamp_9))) {materialFunction=&MF_FX_EmissiveAdjustments}
 ]
}
