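# Unlit masked Niagara material for the trail of the FrogLTE2 vehicle boost, usable on ribbons, sprites and meshes.
# A panning trail texture, driven by dynamic speed and offset, builds a yellow shimmer plus a particle-colored layer eroded by panning noise.
# That color feeds an emissive adjustment function.
# It is also multiplied by a hard smoothstep cutoff of the noise shaped by a U-axis falloff at both ends, then saturated.
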Material /VehicleCosmetics/Boosts/Boost_FrogLTE2/Material/M_Boost_FrogLTE2_FrogTrail02
graph {
 properties {
  bAutomaticallySetUsageInEditor=false
  bUsedWithNiagaraMeshParticles=true
  bUsedWithNiagaraRibbons=true
  bUsedWithNiagaraSprites=true
  blendMode=BLEND_Masked
  shadingModel=MSM_Unlit
  usedShadingModels=Unlit
 }
 nodes [
  n1 = Constant() {r=17}
  n2 = DynamicParameter() {paramNames [speed, offSET, offset mask, erode]}
  n3 = AppendVector(A=@n2.offSET, B=ScalarParameter() {parameterName=append param})
  n4 = TextureCoordinate() {uTiling=1.5}
  n5 = Time()
  n6 = TextureSample(UVs=Panner(Coordinate=Add(A=@n3, B=@n4), Time=Multiply(A=@n5, B=@n2.speed)) {speedX=-2}) {texture=/DealershipSharedLibrary/FX/Textures/TexturePackage/T_VCVFX_FrogTrail01_RGB}
  n7 = Constant() {r=1}
  n8 = Power(Base=Multiply(A=Multiply(A=@n6.G, B=@n7), B=@n6.G), Exp=Constant() {r=0.5})
  MaterialExpressionConstantClamp_1 = Saturate(None=Multiply(A=Multiply(A=@n1, B=@n6.G), B=DotProduct(A=@n8, B=@n6.R)))
  n9 = ParticleColor()
  n10 = Time()
  n11 = ScalarParameter() {defaultValue=1, parameterName=NoiseSpeed}
  n12 = TextureSample(UVs=Panner(Coordinate=TextureCoordinate(), Time=Multiply(A=@n10, B=@n11)) {speedX=-2}) {texture=/DealershipSharedLibrary/FX/Textures/Diffuse/T_VCVFX_Frog_Shimmer_RGB}
  MaterialExpressionConstantClamp_2 = Saturate(None=OneMinus(None=Subtract(A=@n12.G, B=@n2.erode)))
  n13 = Add(
   A=Multiply(
    A=VectorParameter() {parameterName=yellow shimmer, defaultValue=<rgba 0.6077 0.8202 0 0>}.RGB
    B=Multiply(A=@MaterialExpressionConstantClamp_1, B=@n9.A)
   )
   B=Multiply(A=Multiply(A=@n9.RGB, B=@n6.R), B=@MaterialExpressionConstantClamp_2)
  )
  n14 = Constant()
  n15 = Add(A=@n14, B=Constant() {r=0.01})
  n16 = Constant() {r=1.5}
  n17 = ComponentMask(None=TextureCoordinate()) {r=true}
  n18 = Power(Base=OneMinus(None=@n17), Exp=Constant() {r=3.3})
  n19 = Constant() {r=1}
  n20 = Power(Base=@n17, Exp=Constant() {r=2})
  n21 = Constant() {r=3}
  n22 = Multiply(A=Multiply(A=Power(Base=@n12.G, Exp=@n16), B=Constant() {r=0.8}), B=Multiply(A=Multiply(A=@n18, B=@n19), B=Multiply(A=@n20, B=@n21)))
  MaterialExpressionConstantClamp_0 = Saturate(None=Multiply(A=@n13, B=SmoothStep(Min=@n14, Max=@n15, Value=@n22)))
  MaterialFunctionCall(Emissive=@n13) {materialFunction=&MF_FX_EmissiveAdjustments}
 ]
}
